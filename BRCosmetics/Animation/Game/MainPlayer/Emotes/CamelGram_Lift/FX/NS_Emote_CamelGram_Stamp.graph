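# Niagara one-shot stamp impact for the CamelGram Lift emote, a short dark-blue splash at the gadget mesh's JNT_Peace bone.
# Droplets and Splatter burst sprites from a ring around the sampled bone, pushing them outward with drag while they shrink and fade.
# WideRipples adds a growing, rotating mesh, with a color-scale step reading user intensity, time-of-day color and frontend alpha before a blue color curve.
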NiagaraSystem /BRCosmetics/Animation/Game/MainPlayer/Emotes/CamelGram_Lift/FX/NS_Emote_CamelGram_Stamp
dependencies {
 usedDataInterfaces [
  &NiagaraDataInterfaceSkeletalMesh
  &NiagaraDataInterfaceCurve
  &NiagaraDataInterfaceColorCurve
  &NiagaraDataInterfaceVector2DCurve
 ]
 usedDynamicInputs [
  &RandomRangeVector
  &MakeCustomFloatFromBool
  &FNTimeOfDayColor
 ]
 usedModules [
  &SystemState
  &self:SystemSpawnScript.SystemScriptSource.SystemScriptGraph.NiagaraNodeAssignment_0.SetVariables_#1
  &self:SystemSpawnScript.SystemScriptSource.SystemScriptGraph.NiagaraNodeAssignment_1.SetVariables_#2
  &self:SystemSpawnScript.SystemScriptSource.SystemScriptGraph.NiagaraNodeAssignment_2.SetVariables_#3
 ]
}
emitters {
 Droplets {
  emitterUpdateScript [
   &EmitterState()
   &SpawnBurst_Instantaneous(Spawn Count=10, Spawn Group=1, Use Spawn Probability=false, UseLoopCountLimit=false)
  ]
  particleSpawnScript [
   &SkeletalMeshLocation(
    Bone Sampling Mode=::NewEnumerator4
    MaskSpawnGroup=false
    Position Sampling=::NewEnumerator2
    Skeletal Mesh=DataInterface() {propertyValues {
      DefaultMesh=/BRCosmetics/Animation/Props/Gadgets_CamelGram_Lift/Meshes/CamelGram_Lift_Gadget
      ExcludeBoneName=root
      OptionalComponentClass=null
      PreviewMesh=&CamelGram_Lift_Gadget
      SkinningMode=SkinOnTheFly
      SoftSourceActor=null
      SourceMode=Default
      UvSetIndex=0
      WholeMeshLOD=-1
      bExcludeBone=false
      bReadDeformedGeometry=true
      bRequireCurrentFrameData=true
      ComponentTags []
      FilteredBones [JNT_Peace]
      FilteredSockets []
      MeshUserParameter {parameter {name=None, typeDefHandle {registeredTypeIndex=1}, varData []}}
      SamplingRegions []
     }}
   )
   &InitializeParticle(
    Color=<rgba 0 0.0079 0.0365 1>
    Lifetime Max=0.3
    Lifetime Min=0.2
    Lifetime Mode=::NewEnumerator1
    Mesh Scale Mode=::NewEnumerator3
    Sprite Rotation Mode=::NewEnumerator2
    Sprite Size Mode=::NewEnumerator1
    Uniform Sprite Size Max=4
    Uniform Sprite Size Min=2
    UsePositionOffset=true
    Write Material Random=true
   ) {module=InitializeParticle001}
   &ShapeLocation(
    MaskSpawnGroup=false
    Offset=<0 0 -5>
    Offset Mode=::NewEnumerator0
    Override Seed=false
    Ring Radius=15
    Rotation Axis=<0 1 0>
    Rotation Mode=::NewEnumerator1
    Shape Primitive=::NewEnumerator4
    UseOwnerScale=true
   )
   /SetVariables(Particles.Position=&AddVectorToPosition(
     Position=&AddVectorToPosition(
      Position=$Particles.SkeletalMeshLocation.SampledPosition
      Vector=Linked() {linkedVariable {name=Particles.ShapeLocation.ShapePosition, type=NiagaraPosition}}
     )
     Vector=<0 0 0>
    )) {isSetParameters=true}
   /Niagara/Modules/Update/SubUV/SubUVAnimation(SubUV Animation Mode=::NewEnumerator1)
   &AddVelocity(
    Module.Use Falloff Distance=false
    Override Seed=false
    Velocity Mode=::NewEnumerator1
    Velocity Origin=&AddVectorToPosition(Position=$Particles.SkeletalMeshLocation.SampledPosition, Vector=<0 0 0>)
    Velocity Speed=&RandomRangeFloat(Minimum=150, Maximum=250)
   ) {module=AddVelocity001}
   &CameraOffset(Camera Offset Amount=-25)
  ]
  particleUpdateScript [
   &ParticleState()
   &Drag(Drag=&RandomRangeFloat(Minimum=9, Maximum=12), Use Linear Drag=true, Use Rotational Drag=false)
   &SolveForcesAndVelocity(Clamp Velocity=false, Limit Acceleration=false)
   &ScaleSpriteSize(Scale Factor=&Vector2DFromCurve(
     Vector2Curve=DataInterface() {propertyValues {
       XCurve {keys [<at 0 1>,<at 0.5 1>,<at 0.7 1>,<at 1 0>]}
       YCurve {keys [
         <at 0 3 RCIM_Cubic>
         {arriveTangent=0.2942, interpMode=RCIM_Cubic, leaveTangent=0.2942, tangentMode=RCTM_User, time=0.5, value=1}
         {arriveTangent=0.058, interpMode=RCIM_Cubic, leaveTangent=0.058, tangentMode=RCTM_User, time=0.7, value=1}
         <at 1 0 RCIM_Cubic>
        ]}
      }}
     ScaleCurve=<0.6 1>
    ))
   &ScaleColor(
    Linear Color Curve=DataInterface() {propertyValues {
      AlphaCurve {keys [<at 0 0>,<at 0.15 1>,<at 0.5 1>,<at 1 0>]}
      BlueCurve {keys [<at 0 5>,<at 1 0.2>]}
      GreenCurve {keys [<at 0 5>,<at 1 0.2>]}
      RedCurve {keys [<at 0 5>,<at 1 0.2>]}
     }}
    Scale Mode=::NewEnumerator2
   ) {module=ScaleColor001}
  ]
  rendererClasses [&NiagaraSpriteRendererProperties]
  renderers [{rendererClass=&NiagaraSpriteRendererProperties}]
 }
 Splatter {
  emitterUpdateScript [
   &EmitterState()
   &SpawnBurst_Instantaneous(Spawn Count=10, Spawn Group=1, Use Spawn Probability=false, UseLoopCountLimit=false)
  ]
  particleSpawnScript [
   &SkeletalMeshLocation(
    Bone Sampling Mode=::NewEnumerator4
    MaskSpawnGroup=false
    Position Sampling=::NewEnumerator2
    Skeletal Mesh=DataInterface() {propertyValues {
      DefaultMesh=&CamelGram_Lift_Gadget
      ExcludeBoneName=root
      OptionalComponentClass=null
      PreviewMesh=&CamelGram_Lift_Gadget
      SkinningMode=SkinOnTheFly
      SoftSourceActor=null
      SourceMode=Default
      UvSetIndex=0
      WholeMeshLOD=-1
      bExcludeBone=false
      bReadDeformedGeometry=true
      bRequireCurrentFrameData=true
      ComponentTags []
      FilteredBones [JNT_Peace]
      FilteredSockets []
      MeshUserParameter {parameter {name=None, typeDefHandle {registeredTypeIndex=1}, varData []}}
      SamplingRegions []
     }}
   )
   &InitializeParticle(
    Color=<rgba 0 0.0065 0.1667 1>
    Lifetime Max=0.2
    Lifetime Min=0.15
    Lifetime Mode=::NewEnumerator1
    Mesh Scale Mode=::NewEnumerator3
    Sprite Rotation Mode=::NewEnumerator2
    Sprite Size Max=<10 50>
    Sprite Size Min=<5 35>
    Sprite Size Mode=::NewEnumerator2
    UsePositionOffset=true
    Write Material Random=true
   ) {module=InitializeParticle001}
   &ShapeLocation(
    MaskSpawnGroup=false
    Offset=<0 0 -5>
    Offset Mode=::NewEnumerator0
    Override Seed=false
    Ring Radius=15
    Rotation Axis=<0 1 0>
    Rotation Mode=::NewEnumerator1
    Shape Primitive=::NewEnumerator4
    UseOwnerScale=true
   )
   /SetVariables(Particles.Position=&AddVectorToPosition(
     Position=&AddVectorToPosition(
      Position=$Particles.SkeletalMeshLocation.SampledPosition
      Vector=Linked() {linkedVariable {name=Particles.ShapeLocation.ShapePosition, type=NiagaraPosition}}
     )
     Vector=<0 0 0>
    )) {isSetParameters=true}
   &SubUVAnimation(SubUV Animation Mode=::NewEnumerator1)
   &CameraOffset(Camera Offset Amount=-20)
   &AddVelocity(
    Module.Use Falloff Distance=false
    Override Seed=false
    Velocity Mode=::NewEnumerator1
    Velocity Origin=&AddVectorToPosition(Position=$Particles.SkeletalMeshLocation.SampledPosition, Vector=<0 0 0>)
    Velocity Speed=200
   )
  ]
  particleUpdateScript [
   &ParticleState()
   &SpriteRotationRate(Rotation Rate=&RandomRangeFloat(Minimum=-90, Maximum=90))
   &Drag(Drag=&RandomRangeFloat(Minimum=9, Maximum=12), Use Linear Drag=true, Use Rotational Drag=false)
   &SolveForcesAndVelocity(Clamp Velocity=false, Limit Acceleration=false)
   &ScaleSpriteSize(Scale Factor=&Vector2DFromCurve(
     Vector2Curve=DataInterface() {propertyValues {
       XCurve {keys [<at 0 1>,<at 1 1>]}
       YCurve {keys [<at 0 1 RCIM_Cubic>, {arriveTangent=-3.2254, interpMode=RCIM_Cubic, leaveTangent=-3.2254, tangentMode=RCTM_User, time=1, value=0}]}
      }}
     ScaleCurve=<1 1>
    ))
   &ScaleColor(
    Linear Color Curve=DataInterface() {propertyValues {
      AlphaCurve {keys [<at 0 0>,<at 0.15 1>,<at 0.5 1>,<at 1 0>]}
      BlueCurve {keys [<at 0 5>,<at 1 0.2>]}
      GreenCurve {keys [<at 0 5>,<at 1 0.2>]}
      RedCurve {keys [<at 0 5>,<at 1 0.2>]}
     }}
    Scale Mode=::NewEnumerator2
   ) {module=ScaleColor001}
   &DynamicMaterialParameters(
    Index 0 Param 1=&FloatFromCurve(FloatCurve=DataInterface() {propertyValues {Curve {keys [
         {arriveTangent=0.0674, interpMode=RCIM_Cubic, leaveTangent=0.0674, tangentMode=RCTM_User, time=0, value=1}
         {arriveTangent=-2.8897, interpMode=RCIM_Cubic, leaveTangent=-2.8897, tangentMode=RCTM_User, time=1, value=0}
        ]}}})
    Index 2 Param 4=1.2
    Module.Param0WriteEnabled=true
    Module.Param0WriteEnabled2=true
    Module.Param1WriteEnabled=false
    Module.Param1WriteEnabled2=true
    Module.Param2WriteEnabled=false
    Module.Param2WriteEnabled2=true
    Module.Param3WriteEnabled=false
    Module.Param3WriteEnabled2=true
    Write Parameter Index 2=true
   )
  ]
  rendererClasses [&NiagaraSpriteRendererProperties]
  renderers [
   {rendererClass=&NiagaraSpriteRendererProperties}
   {rendererClass=&NiagaraSpriteRendererProperties, rendererIndex=1}
  ]
 }
 WideRipples {
  emitterUpdateScript [
   &EmitterState()
   &SpawnBurst_Instantaneous(Use Spawn Probability=false, UseLoopCountLimit=false)
  ]
  particleSpawnScript [
   &SkeletalMeshLocation(
    Bone Sampling Mode=::NewEnumerator4
    MaskSpawnGroup=false
    Position Sampling=::NewEnumerator2
    Skeletal Mesh=DataInterface() {propertyValues {
      DefaultMesh=&CamelGram_Lift_Gadget
      ExcludeBoneName=root
      OptionalComponentClass=null
      PreviewMesh=&CamelGram_Lift_Gadget
      SkinningMode=SkinOnTheFly
      SoftSourceActor=null
      SourceMode=Default
      UvSetIndex=0
      WholeMeshLOD=-1
      bExcludeBone=false
      bReadDeformedGeometry=true
      bRequireCurrentFrameData=true
      ComponentTags []
      FilteredBones [JNT_Peace]
      FilteredSockets []
      MeshUserParameter {parameter {name=None, typeDefHandle {registeredTypeIndex=1}, varData []}}
      SamplingRegions []
     }}
   )
   &InitializeParticle(
    Color=<rgba 1 1 1 15>
    Lifetime=0.2
    Mesh Scale Mode=::NewEnumerator3
    Mesh Uniform Scale=125
    Sprite Rotation Mode=::NewEnumerator2
    Sprite Size Mode=::NewEnumerator3
    Uniform Sprite Size=1500
    UsePositionOffset=false
    Write Material Random=true
   )
   &InitialMeshOrientation(Mesh Orientation Mode=::NewEnumerator2, Use Rotation Vector=false)
   /SetVariables(Particles.Position=&AddVectorToPosition(Position=$Particles.SkeletalMeshLocation.SampledPosition, Vector=<0 0 0>)) {isSetParameters=true}
   &CameraOffset(Camera Offset Amount=5)
  ]
  particleUpdateScript [
   &ParticleState()
   &UpdateMeshOrientation(Rotation Rate=0.2, Rotation Vector=<0 0 1>)
   &ScaleMeshSize(Scale Factor=&VectorFromFloat(Value=&FloatFromCurve(
      FloatCurve=DataInterface() {propertyValues {Curve {keys [{tangentMode=RCTM_User, time=0, value=0.4}, {tangentMode=RCTM_User, time=1, value=1}]}}}
      Scale Curve=0.6
     )))
   &ScaleColor(
    Scale Alpha=&FloatFromCurve(
     FloatCurve=DataInterface() {propertyValues {Curve {keys [
         {arriveTangent=13.2248, interpMode=RCIM_Cubic, leaveTangent=13.2248, tangentMode=RCTM_User, time=0, value=0}
         <at 0.1113 0.9882 RCIM_Cubic>
         {interpMode=RCIM_Cubic, tangentMode=RCTM_User, time=1, value=0}
        ]}}}
     Scale Curve=&Multiply_Float(
      A=&Multiply_Float(A=$User.Intensity, B=$User.ColorIntensity)
      B=&Multiply_Float(A=$System.AgeFade, B=$System.FrontEnd_Alpha)
     )
    )
    Scale RGB=&Multiply_Vector(A=<1 1 1>, B=&MakeVectorFromLinearColorRGB(Color=$System.ToD))
    ScaleA=true
    ScaleRGB=true
   )
   &SolveForcesAndVelocity(Clamp Velocity=false, Limit Acceleration=false)
   &DynamicMaterialParameters(
    Index 0 Param 1=0.5
    Index 0 Param 2=1
    Index 0 Param 3=1
    Index 0 Param 4=1
    Module.Param0WriteEnabled=true
    Module.Param1WriteEnabled=true
    Module.Param2WriteEnabled=true
    Module.Param3WriteEnabled=true
   )
   &Color(Color=&ColorFromCurve(ColorCurve=DataInterface() {propertyValues {
       AlphaCurve {keys [<at 0 8 RCIM_Cubic>,<at 1 0 RCIM_Cubic>]}
       BlueCurve {keys [<at 0 0.2917 RCIM_Cubic>]}
       GreenCurve {keys [<at 0 0.0374 RCIM_Cubic>]}
       RedCurve {keys [<at 0 0 RCIM_Cubic>]}
      }}))
  ]
  rendererClasses [&NiagaraMeshRendererProperties]
  renderers [{rendererClass=&NiagaraMeshRendererProperties, rendererIndex=0}]
 }
}
summary {
 systemName=NS_Emote_CamelGram_Stamp
 emitters [
  {emitterName=WideRipples, rendererClasses [&NiagaraMeshRendererProperties]}
  {emitterName=Splatter, rendererClasses [&NiagaraSpriteRendererProperties]}
  {emitterName=Droplets, rendererClasses [&NiagaraSpriteRendererProperties]}
 ]
 userVariables [
  User.bIsFrontendPreview {type=NiagaraBool, defaultValue=true}
  User.ColorIntensity {type=NiagaraFloat, defaultValue=2, description=""}
  User.Intensity {type=NiagaraFloat, defaultValue=1, description=""}
 ]
}
